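# Unlit material for the disc part of the DiscoDisc vehicle wheel, faking a view-dependent radial reflection.
# Centered UVs become an angle via Arctangent2, which forms a vector with its inverse and is transformed out of view space, so the pattern shifts with the camera.
# The resulting XY samples a reflection mask texture as emissive color.
# Material attributes then pass through the shared vehicle damage and de-rez transition functions.
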Material /VehicleCosmetics/Wheels/DiscoDisc/Materials/M_Wheel_DiscoDisc_DiscMat
graph {
 properties {
  bAutomaticallySetUsageInEditor=false
  bUseMaterialAttributes=true
  bUsedWithSkeletalMesh=true
  previewMesh=/VehicleCosmetics/Wheels/DiscoDisc/Meshes/SK_Wheel_DiscoDisc
  shadingModel=MSM_Unlit
  usedShadingModels=Unlit
 }
 nodes [
  n1 = Subtract(A=TextureCoordinate() {uTiling=2, vTiling=2}, B=Constant() {r=1})
  n2 = ComponentMask(None=@n1) {g=true}
  n3 = ComponentMask(None=@n1) {r=true}
  n4 = Divide(
   A=ComponentMask(None=Arctangent2!(Y=@n2, X=@n3)) {r=true}
   B=MaterialFunctionCall(Multiplier=Constant() {r=2}) {materialFunction=&Pi}.Result
  )
  n5 = OneMinus(None=@n4)
  n6 = AppendVector(A=AppendVector(A=@n4, B=@n5), B=@n5)
  n7 = TextureSample(UVs=ComponentMask(None=Transform(None=@n6) {transformSourceType=TRANSFORMSOURCE_View}) {g=true, r=true}) {texture=/VehicleCosmetics/Wheels/DiscoDisc/Textures/T_DiscoDisc_ReflectionMask}
  MaterialFunctionCall(In=MaterialFunctionCall(Material=SetMaterialAttributes(Emissive Color=@n7.RGB) {attributeSetTypes [#1]}) {materialFunction=&MF_BaseVehicle_Damage}.Result) {materialFunction=&MF_VehicleDeRez_Transition}
 ]
}
